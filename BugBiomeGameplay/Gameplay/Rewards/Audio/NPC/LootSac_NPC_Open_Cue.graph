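# Sound cue for an NPC loot sac opening, from the bug biome's reward audio.
# A quality-level root sends its first two levels to a random pick of five open sounds and its last two to the first sound alone, each path through a modulator for slight variation.
# It plays boosted in the loot reward sound class with item-interact attenuation, and a concurrency override prevents new instances once the limit is reached.
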SoundCue /BugBiomeGameplay/Gameplay/Rewards/Audio/NPC/LootSac_NPC_Open_Cue
root=QualityLevel_0
nodes [
 {id=Modulator_2, children [WavePlayer_5], properties {pitchMin=1}}
 {id=Modulator_3, children [Random_0], properties {pitchMin=1, volumeMax=1, volumeMin=1}}
 {id=QualityLevel_0, children [Modulator_3,Modulator_3,Modulator_2,Modulator_2]}
 {id=Random_0, children [WavePlayer_0,WavePlayer_2,WavePlayer_3,WavePlayer_4,WavePlayer_6], properties {weights [1,1,1,1,1]}}
 {id=WavePlayer_0, properties {soundWaveAssetPtr=/BugBiomeGameplay/Gameplay/Rewards/Audio/Infest_LootSac_Open_01}}
 {id=WavePlayer_2, properties {soundWaveAssetPtr=/BugBiomeGameplay/Gameplay/Rewards/Audio/Infest_LootSac_Open_02}}
 {id=WavePlayer_3, properties {soundWaveAssetPtr=/BugBiomeGameplay/Gameplay/Rewards/Audio/Infest_LootSac_Open_03}}
 {id=WavePlayer_4, properties {soundWaveAssetPtr=/BugBiomeGameplay/Gameplay/Rewards/Audio/Infest_LootSac_Open_04}}
 {id=WavePlayer_5, properties {soundWaveAssetPtr=&Infest_LootSac_Open_01}}
 {id=WavePlayer_6, properties {soundWaveAssetPtr=/BugBiomeGameplay/Gameplay/Rewards/Audio/Infest_LootSac_Open_05}}
]
properties {
 attenuationSettings=/Game/Sounds/Attenuation/Fort_Item_Interact_Att
 bOverrideConcurrency=true
 maxDistance=1850
 soundClassObject=&Fort_Loot_Reward_PSM
 volumeMultiplier=2.8
 concurrencyOverrides {resolutionRule=PreventNew}
}
