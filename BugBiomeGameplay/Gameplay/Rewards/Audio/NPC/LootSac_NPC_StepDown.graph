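# Sound cue for the loot sac NPC's step-down sound in the bug biome rewards content.
# A random node picks one of three equally weighted jump-down waves and feeds it into a modulator that holds volume steady and keeps pitch from dropping below normal.
# It uses item-interact attenuation and the loot reward sound class, with a concurrency override that prevents new instances once the limit is reached.
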SoundCue /BugBiomeGameplay/Gameplay/Rewards/Audio/NPC/LootSac_NPC_StepDown
root=Modulator_3
nodes [
 {id=Modulator_3, children [Random_0], properties {pitchMin=1, volumeMax=1, volumeMin=1}}
 {id=Random_0, children [WavePlayer_1,WavePlayer_7,WavePlayer_8], properties {weights [1,1,1]}}
 {id=WavePlayer_1, properties {soundWaveAssetPtr=&/Infest_LootSac_NPC_JumpDown_01}}
 {id=WavePlayer_7, properties {soundWaveAssetPtr=&/Infest_LootSac_NPC_JumpDown_02}}
 {id=WavePlayer_8, properties {soundWaveAssetPtr=&/Infest_LootSac_NPC_JumpDown_03}}
]
properties {
 attenuationSettings=/Game/Sounds/Attenuation/Fort_Item_Interact_Att
 bOverrideConcurrency=true
 maxDistance=1850
 soundClassObject=&Fort_Loot_Reward_PSM
 volumeMultiplier=1
 concurrencyOverrides {resolutionRule=PreventNew}
}
